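# M_Rope is a Default Lit material for Niagara ribbons, made for the rope effect of the NyxGlass item.
# It samples the gadget's diffuse and normal textures alongside Fresnel, UV edge-mask and UV threshold branches.
# Sky-light direction dotted with pixel normals builds a sharp particle-colored highlight weighted along V.
# A sine band along U, offset by a dynamic parameter, forms another mask; no material output connections appear.
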Material /DryBox/Items/NyxGlass/FX/Assets/M_Rope
graph {
 properties {bUsedWithNiagaraRibbons=true, pixelDepthOffsetMode=PDOM_Legacy, usedShadingModels=Default Lit}
 nodes [
  Constant() {r=1}
  n1 = ScalarParameter() {defaultValue=2}
  Fresnel(ExponentIn=@n1) {baseReflectFraction=0}
  n2 = MaterialFunctionCall(Float2=TextureCoordinate()) {materialFunction=&BreakOutFloat2Components}
  n3 = Constant()
  If(A=@n2.G, B=Constant() {r=0.2}, "A > B"=Constant() {r=1}, "A == B"=@n3, "A < B"=@n3)
  n4 = VectorParameter() {group=Opacity, parameterName=Opacity_EdgeMask_Control, channelNames {a="", b=Power, g=Multiply, r=Offset}, defaultValue=<rgba 0 1 1 1>}
  MaterialFunctionCall!(
   UorV=MaterialFunctionCall(Float2=TextureCoordinate()) {materialFunction=&BreakOutFloat2Components}.R
   EdgeControl=@n4
  ) {materialFunction=/Game/Effects/Niagara/MasterMaterials/Functions/Common/FXCommon_EdgeMask_Simple}
  n5 = Saturate(None=DotProduct(A=SkyAtmosphereLightDirection(), B=PixelNormalWS()))
  n6 = ScalarParameter() {defaultValue=10, parameterName=Param_1}
  n7 = ParticleColor()
  n8 = VectorParameter() {defaultValue=<rgba 5 8.3333 10 1>}
  Multiply(A=@n2.G, B=Multiply(A=Power(Base=@n5, Exp=@n6), B=Multiply(A=@n7.RGB, B=@n8.RGB)))
  OneMinus(None=Power(Base=@n5, Exp=@n1))
  n9 = Add(
   A=MaterialFunctionCall(Float2=TextureCoordinate()) {materialFunction=&BreakOutFloat2Components}.R
   B=Multiply(A=DynamicParameter().Param2, B=ScalarParameter() {defaultValue=3, parameterName=Param_2})
  )
  n10 = Saturate(None=Sine(None=Multiply(A=@n9) {constB=0.15}))
  Saturate(None=Add(A=Power(Base=@n10) {constExponent=6}, B=ScalarParameter() {parameterName=Param_3}))
  TextureSample() {texture=/DryBox/Items/NyxGlass/Textures/T_Gadget_NyxGlass_D_FX}
  TextureSample() {samplerType=SAMPLERTYPE_Normal, texture=/DryBox/Items/NyxGlass/Textures/T_Gadget_NyxGlass_N_FX}
 ]
}
