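# Additive unlit material for Niagara ribbons that draws the NyxGlass rope glow, tinted by a color parameter and particle color.
# A sampled directional noise texture, with UV controls and add, multiply and power adjustments, breaks up the opacity.
# That noise is multiplied by an edge mask across U, an inverted V falloff, a blend of two powered V gradients on offset UVs, and particle alpha.
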Material /DryBox/Items/NyxGlass/FX/Assets/M_RopeGlow
graph {
 properties {bUsedWithNiagaraRibbons=true, blendMode=BLEND_Additive, pixelDepthOffsetMode=PDOM_Legacy, shadingModel=MSM_Unlit, usedShadingModels=Unlit}
 nodes [
  MaterialFunctionCall(Emissive=Multiply(A=VectorParameter() {defaultValue=<rgba 2.075 2.3583 2.5 1>}.RGB, B=ParticleColor().RGB)) {materialFunction=&MF_FX_EmissiveAdjustments}
  n1 = TextureObjectParameter() {
   group=MainTex
   parameterName=MainTex_Texture
   samplerType=SAMPLERTYPE_LinearGrayscale
   sortPriority=0
   texture=/Game/Effects/Fort_Effects/Emotes/UltraLight/Texture/T_DirectionalNoise_Greyscale
  }
  n2 = VectorParameter() {group=MainTex, parameterName=MainTex_UVControl, channelNames {a=VPan, b=UPan, g=VScale, r=UScale}, defaultValue=<rgba 0.2 1 5 0.1>}
  n3 = Add(
   A=MaterialFunctionCall(Texture=@n1, UVControl_RGB=@n2, UVControl_A=@n2) {materialFunction=&FXCommon_TextureSample}.R
   B=ScalarParameter() {defaultValue=0.1, group=MainTex, parameterName=MainTex_GrayAdjust_Add}
  )
  n4 = ScalarParameter() {defaultValue=2, group=MainTex, parameterName=MainTex_GrayAdjust_Mul}
  n5 = Power(Base=Saturate(None=Multiply(A=@n3, B=@n4)), Exp=ScalarParameter() {defaultValue=1, group=MainTex, parameterName=MainTex_GrayAdjust_Power})
  n6 = Constant()
  n7 = ScalarParameter() {defaultValue=-0.1}
  n8 = MaterialFunctionCall(Float2=Add(A=TextureCoordinate(), B=AppendVector(A=@n6, B=@n7))) {materialFunction=&BreakOutFloat2Components}
  n9 = Power(Base=OneMinus(None=@n8.G))
  n10 = VectorParameter() {group=RibbonB, parameterName=Ribbon1_EdgeMask_InvertControl, channelNames {a="", b=Power, g=Multiply, r=Offset}, defaultValue=<rgba 0 1 3 1>}
  n11 = VectorParameter() {group=RibbonB, parameterName=Ribbon1_EdgeMask_Control, channelNames {a="", b=Power, g=Multiply, r=Offset}, defaultValue=<rgba 0 3 1.5 1>}
  n12 = ScalarParameter() {group=R, parameterName=R_GrayAdjust_Add}
  n13 = Saturate(None=Multiply(A=Add(A=@n8.G, B=@n12), B=ScalarParameter() {defaultValue=2, group=R, parameterName=R_GrayAdjust_Mul}))
  n14 = VectorParameter() {parameterName=Ref, defaultValue=<rgba 25 1 2 0.3>}
  n15 = Power(Base=@n13, Exp=@n14.R)
  n16 = Power(Base=@n13, Exp=@n14.B)
  n17 = Multiply(
   A=MaterialFunctionCall!(UorV=@n8.R, Invert_EdgeControl=@n10, NoInvert_EdgeControl=@n11) {materialFunction=/Game/Effects/Niagara/MasterMaterials/Functions/Common/FXCommon_EdgeMask_Ex}.Mask
   B=LinearInterpolate(A=Multiply(A=@n15, B=@n14.G), B=Multiply(A=@n16, B=@n14.A))
  )
  Multiply(A=ParticleColor().A, B=Multiply(A=Saturate(None=@n5), B=Multiply(A=@n9, B=@n17)))
 ]
}
