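# Translucent unlit Niagara ribbon material for NyxGlass effects, built around panning smoke-trail textures.
# One branch shapes a mask from smoke, a curve-atlas row along UV V, particle alpha, inverse Fresnel and camera-distance fade.
# Another fakes lighting by blending Smoke and Lit colors with sky light direction against a normal map.
# Further branches tint by particle color, apply emissive adjustments and threshold a dissolve texture with dynamic parameter D.
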Material /DryBox/Items/NyxGlass/FX/Assets/M_TempRibbon
graph {
 properties {
  bUsedWithNiagaraRibbons=true
  blendMode=BLEND_Translucent
  pixelDepthOffsetMode=PDOM_Legacy
  shadingModel=MSM_Unlit
  translucencyLightingMode=TLM_VolumetricPerVertexNonDirectional
  usedShadingModels=Unlit
 }
 nodes [
  Constant() {r=1}
  n1 = WorldPosition()
  n2 = CameraPositionWS()
  n3 = Saturate(None=Divide(A=Distance(A=@n1.XYZ, B=@n2)) {constB=200})
  n4 = OneMinus(None=Fresnel() {baseReflectFraction=0, exponent=2})
  n5 = TextureCoordinate()
  n6 = VectorParameter() {parameterName=S, defaultValue=<rgba 3 0 0 1>}
  n7 = Panner(Coordinate=TextureCoordinate(), Speed=AppendVector(A=@n6.R, B=@n6.R))
  n8 = ParticleColor()
  n9 = Multiply(
   A=CurveAtlasRowParameter(CurveTime=MaterialFunctionCall(Float2=@n5) {materialFunction=&BreakOutFloat2Components}.G) {atlas=/DryBox/Items/NyxGlass/FX/Air/CLA_AirBreak, curve=/DryBox/Items/NyxGlass/FX/Air/CL_RibbonA, parameterName=C}.R
   B=Multiply(
    A=TextureSample(UVs=@n7) {samplerType=SAMPLERTYPE_Grayscale, texture=/Game/Effects/Fort_Effects/Textures/Smoke/T_Smoke_Trail_Soft_NoClamp}.R
    B=@n8.A
   )
  )
  LinearInterpolate(Alpha=Multiply(A=Power(Base=@n3) {constExponent=3}, B=Multiply(A=@n4, B=@n9))) {constA=1, constB=0}
  MaterialFunctionCall(Emissive=@n8.RGB) {materialFunction=&MF_FX_EmissiveAdjustments}
  n10 = SkyAtmosphereLightDirection()
  n11 = Transform(None=TextureSample(UVs=@n7) {samplerType=SAMPLERTYPE_Normal, texture=/Game/Effects/Fort_Effects/Gameplay/SmokeDustExplosions/Textures/T_RibbonSmokeMasked_01}.RGB)
  n12 = LinearInterpolate(
   A=VectorParameter() {parameterName=Smoke, defaultValue=<rgba 0.7 0.7 0.7 1>}.RGB
   B=VectorParameter() {parameterName=Lit, defaultValue=<rgba 1 1 1 1>}.RGB
   Alpha=Saturate(None=DotProduct(A=@n10, B=@n11))
  )
  n13 = TextureSample(UVs=@n7) {samplerType=SAMPLERTYPE_LinearColor, texture=/Game/Effects/Fort_Effects/Gameplay/SmokeDustExplosions/Textures/T_RibbonSmokeMasked_03}
  Multiply(A=@n12, B=Multiply(A=LinearInterpolate(A=@n13.R), B=@n8.RGB) {constA=1})
  n14 = TextureSample(UVs=@n7) {texture=/Game/Effects/Fort_Effects/Gameplay/SmokeDustExplosions/Textures/T_RibbonSmokeMasked_04}
  n15 = Constant() {r=1}
  Multiply(
   A=If(
    A=DynamicParameter() {defaultValue=<rgba 0.5 1 1 1>, paramNames [D,Param2,Param3,Param4]}.D
    B=OneMinus(None=@n14.R)
    "A > B"=@n15
    "A == B"=@n15
    "A < B"=Constant()
   )
   B=Multiply(A=TextureSample(UVs=@n7) {texture=/Game/Effects/Fort_Effects/Gameplay/SmokeDustExplosions/Textures/T_RibbonSmokeMasked_02}.R, B=@n8.A)
  )
  Multiply(A=@n12) {constB=0.5}
 ]
}
